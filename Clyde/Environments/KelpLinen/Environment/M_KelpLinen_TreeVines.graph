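# Masked two-sided subsurface foliage material for KelpLinen tree vines, built for Nanite and instancing, with output passed through MF_Kelp_AudioReactive.
# Tints desaturated leaf texture separately for leaves and branches, with packed masks for emission, subsurface, roughness and branches.
# Emission combines vertex-red falloffs with a UV gradient whose exponent follows an audio-reactive parameter.
# Sway rotates around instance pivots, randomized per instance and blended by storm mask; older SimpleGrassWind and noise-emission chains remain unconnected.
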Material /Clyde/Environments/KelpLinen/Environment/M_KelpLinen_TreeVines
graph {
 properties {
  bAutomaticallySetUsageInEditor=false
  bUseMaterialAttributes=true
  bUsedWithInstancedStaticMeshes=true
  bUsedWithNanite=true
  blendMode=BLEND_Masked
  shadingModel=MSM_Subsurface
  twoSided=true
  usedShadingModels=Subsurface
 }
 nodes [
  MaterialFunctionCall!(In=MakeMaterialAttributes!(
    BaseColor=NamedRerouteUsage()
    Metallic=Constant()
    Specular=Constant() {r=0.5}
    Roughness=NamedRerouteUsage()
    Anisotropy
    EmissiveColor=NamedRerouteUsage()
    Opacity
    OpacityMask=NamedRerouteUsage()
    Normal=NamedRerouteUsage()
    Tangent
    WorldPositionOffset=NamedRerouteUsage()
    SubsurfaceColor=NamedRerouteUsage()
    ClearCoat
    ClearCoatRoughness
    AmbientOcclusion
    Refraction
    CustomizedUVs_0
    CustomizedUVs_1
    CustomizedUVs_2
    CustomizedUVs_3
    CustomizedUVs_4
    CustomizedUVs_5
    CustomizedUVs_6
    CustomizedUVs_7
    PixelDepthOffset
    ShadingModel
    Displacement
   )) {materialFunction=/ClydeCosmetics/Environments/KelpLinen/MF_Kelp_AudioReactive}
  n1 = MaterialFunctionCall(
   Input=NamedRerouteUsage()
   Input Low=ScalarParameter() {defaultValue=-1, group=Emission, parameterName=Emission low}
   Input High=ScalarParameter() {defaultValue=1, group=Emission, parameterName=Emission High}
  ) {materialFunction=&RemapValue_0-1}
  n2 = Subtract(
   A=Multiply(A=Saturate(None=@n1.Result), B=VectorParameter() {group=Emission, parameterName=Emissive Tint, defaultValue=<rgba 50 8.3556 46.4455 1>}.RGB)
   B=NamedRerouteUsage()
  )
  n3 = NamedRerouteUsage()
  n4 = Multiply(A=Time(), B=ScalarParameter() {defaultValue=0.5, group=Emission, parameterName=Emission Wave Speed})
  n5 = Noise!(World Position=AppendVector(A=Panner(Time=@n4) {speedX=0.2, speedY=0.5}, B=Constant() {r=1}), FilterWidth) {scale=0.1}
  n6 = Subtract(
   A=Multiply(A=@n3, B=ScalarParameter() {defaultValue=0.75, group=Emission, parameterName=Emission Wave Intensity}) {constB=0.75}
   B=NamedRerouteUsage()
  )
  Multiply(A=Multiply(A=Multiply(A=@n2, B=@n3), B=Multiply(A=@n5, B=@n6)), B=NamedRerouteUsage())
  n7 = OneMinus(None=NamedRerouteUsage())
  Multiply(
   A=Multiply(A=@n7, B=ScalarParameter() {defaultValue=0.5, group=Base Color, parameterName=Leaf Tint Intensity})
   B=OneMinus(None=NamedRerouteUsage())
  )
  n8 = ScalarParameter() {defaultValue=1, group=Wind, parameterName=Wind Weight}
  n9 = NamedRerouteUsage()
  n10 = Multiply(A=NamedRerouteUsage()) {constB=1.2}
  n11 = Time()
  n12 = Sine(None=Multiply(A=@n11) {constB=0.5})
  n13 = Sine(None=Multiply(A=@n11) {constB=0.2})
  n14 = Panner(Time=Multiply(A=Time(), B=ScalarParameter() {defaultValue=0.5, group=Emission, parameterName=Wind Modulation Speed})) {speedX=0.1, speedY=0.5}
  n15 = Constant() {r=1}
  n16 = Multiply(A=Multiply(A=Add(A=@n12, B=@n13)) {constB=0.5}, B=Noise(World Position=AppendVector(A=@n14, B=@n15)) {scale=0.4})
  n17 = PerInstanceRandom()
  n18 = Multiply(
   A=VectorParameter() {group=Wind, parameterName=Wind Direction WPO, defaultValue=<rgba 1 0.0133 0 1>}.RGB
   B=Multiply(A=Multiply(A=@n10, B=@n16), B=Multiply(A=@n17))
  )
  n19 = MaterialFunctionCall(Hue Shift Percentage=Constant() {r=0.2}, Texture=@n18) {materialFunction=&HueShift}
  n20 = MaterialFunctionCall(Input=@n17, Input Low=Constant() {r=0.25}, Input High=Constant() {r=1}) {materialFunction=&RemapValue_0-1}
  Multiply(
   A=MaterialFunctionCall!(
    WindIntensity=ScalarParameter() {defaultValue=0.5, group=Wind, parameterName=Wind Intensity}
    WindWeight=Multiply(A=@n8, B=@n9)
    WindSpeed=ScalarParameter() {defaultValue=0.2, group=Wind, parameterName=Wind Speed}
    AdditionalWPO=LinearInterpolate(A=@n18, B=@n19.Result, Alpha=@n20.Result)
   ) {materialFunction=/Engine/Functions/Engine_MaterialFunctions01/WorldPositionOffset/SimpleGrassWind}.Result
   B=NamedRerouteUsage()
  )
  NamedRerouteDeclaration(None=VertexColor().R) {name=VC Red, nodeColor=<rgba 1 0 0.0367 1>}
  n21 = TextureCoordinate()
  n22 = TextureSampleParameter2D(UVs=@n21) {group=Textures, parameterName=Base Color Texture, texture=/Game/Environments/Apollo/Foliage/Medium_Tree/Textures/nanite/T_LEAFS_Magical}
  n23 = Desaturation(None=@n22.RGB)
  n24 = VectorParameter() {group=Base Color, parameterName=Branch Tint, defaultValue=<rgba 0.2083 0.0629 0.1812 1>}
  n25 = NaniteReplace!(
   Default=VectorParameter() {group=Base Color, parameterName=Leaf Tint Falback, defaultValue=<rgba 1.7902 0 1.8 1>}.RGB
   Nanite=VectorParameter() {group=Base Color, parameterName=Leaf Tint, defaultValue=<rgba 1.5 0.3457 1.3171 1>}.RGB
  )
  NamedRerouteDeclaration(None=LinearInterpolate(A=Multiply(A=@n23, B=@n24.RGB), B=Multiply(A=@n25, B=@n23), Alpha=@n7)) {name=Base Color, nodeColor=<rgba 0 0.7187 1 1>}
  NamedRerouteDeclaration(None=@n22.A) {name=Opacity Mask, nodeColor=<rgba 1 0.3984 0 1>}
  NamedRerouteDeclaration(None=@n17) {name=PerInstanceRandom, nodeColor=<rgba 0.9688 1 0 1>}
  n26 = TextureSampleParameter2D(UVs=@n21) {
   group=Textures
   parameterName=Normal Texture
   samplerType=SAMPLERTYPE_Normal
   texture=/Game/Environments/Apollo/Foliage/Medium_Tree/Textures/nanite/T_LEAFS_N
  }
  NamedRerouteDeclaration(None=@n26.RGB) {name=Normal, nodeColor=<rgba 0.307 0.2658 1 1>}
  n27 = TextureSampleParameter2D(UVs=@n21) {
   desc="Emission, Subsurface, Roughness, Branch Mask"
   group=Textures
   parameterName=Mask Texture
   samplerType=SAMPLERTYPE_Masks
   texture=/Game/Environments/Apollo/Foliage/Medium_Tree/Textures/nanite/T_LEAFS_ESSR
  }
  NamedRerouteDeclaration(None=@n27.R) {name=Emission Mask, nodeColor=<rgba 0.6406 0 1 1>}
  NamedRerouteDeclaration(None=@n27.B) {name=Roughness Mask, nodeColor=<rgba 0.1573 0.153 0.1527 1>}
  NamedRerouteDeclaration(None=@n27.G) {name=SS Mask, nodeColor=<rgba 1 0.4193 0 1>}
  NamedRerouteDeclaration(None=@n27.A) {name=Branch Mask, nodeColor=<rgba 0.2656 0.1164 0.0403 1>}
  n28 = NamedRerouteUsage()
  n29 = ScalarParameter() {defaultValue=0.5, group=Subsurface, parameterName=Subsurface Mask Strength}
  NamedRerouteDeclaration(None=Multiply(A=NamedRerouteUsage(), B=Multiply(A=@n28, B=@n29))) {name=SS Out, nodeColor=<rgba 0.6406 1 0 1>}
  n30 = Power(Base=VertexColor().R, Exp=ScalarParameter() {defaultValue=2.0918, parameterName=AnimFalloff})
  n31 = PerInstanceRandom()
  n32 = Add(A=Time(), B=Multiply(A=@n31) {constB=33.3})
  n33 = Add(A=Multiply(A=@n31) {constB=0.05}, B=ScalarParameter() {defaultValue=0.1, parameterName=Speed})
  n34 = Multiply(A=@n32, B=Multiply(A=@n33) {constB=0.33})
  n35 = Add(A=Multiply(A=Sine(None=@n34)) {constB=0.5}, B=Sine(None=Multiply(A=@n32, B=@n33)))
  n36 = PerInstanceRandom()
  n37 = Add(A=Time(), B=Multiply(A=@n36) {constB=33.3})
  n38 = Add(A=Multiply(A=@n36) {constB=0.05}, B=ScalarParameter() {defaultValue=1, parameterName=Speed_Storm})
  n39 = Multiply(A=@n37, B=Multiply(A=@n38) {constB=0.33})
  n40 = Add(A=Multiply(A=Sine(None=@n39)) {constB=0.5}, B=Sine(None=Multiply(A=@n37, B=@n38)))
  n41 = MaterialFunctionCall!(Position, IsTrees, IsGrass, IsMediumPlant, IsMediumPlant_2024Anim, IsMediumPlant2ndLevel, MediumPlant_BasePivotPostion) {materialFunction=/Game/Environments/Asteria/Foliage/Wind/MF_WindActorData}
  n42 = Multiply(
   A=Add(A=LinearInterpolate(A=@n35, B=@n40, Alpha=@n41.StormMask), B=ScalarParameter() {defaultValue=1, parameterName=Rotation_Bias})
   B=ScalarParameter() {defaultValue=0.02, parameterName=Amplitude}
  )
  n43 = Constant3Vector()
  NamedRerouteDeclaration(None=RotateAboutAxis(
    NormalizedRotationAxis=VectorParameter() {parameterName=WindDirection, defaultValue=<rgba 0 1 0 1>}.RGB
    RotationAngle=Multiply(A=@n30, B=@n42)
    PivotPoint=TransformPosition(None=@n43) {transformSourceType=TRANSFORMPOSSOURCE_Instance, transformType=TRANSFORMPOSSOURCE_World}
    Position=WorldPosition().XYZ
   )) {name=WPO Out, nodeColor=<rgba 1 0 0.9609 1>}
  n44 = ComponentMask(None=OneMinus(None=TextureCoordinate())) {g=true}
  n45 = PerInstanceRandom()
  n46 = MaterialFunctionCall() {materialFunction=/ClydeCosmetics/Environments/KelpLinen/MF_Kelp_AudioReactive_Parameter}
  n47 = LinearInterpolate(
   A=ScalarParameter() {defaultValue=50, parameterName=Reactive_Low}
   B=ScalarParameter() {defaultValue=1, parameterName=Reactive_High}
   Alpha=Add(A=LinearInterpolate(Alpha=@n45) {constA=-0.2, constB=0}, B=Clamp(None=@n46.Result))
  ) {constA=1}
  n48 = VertexColor()
  n49 = ScalarParameter() {defaultValue=50, parameterName=EmissiveFalloffTip}
  n50 = NaniteReplace(
   Default=ScalarParameter() {defaultValue=2, parameterName=EmissiveFalloff_Fallback}
   Nanite=ScalarParameter() {defaultValue=53.3607, parameterName=EmissiveFalloff}
  )
  n51 = Multiply(
   A=Power(Base=NamedRerouteUsage(), Exp=ScalarParameter() {defaultValue=0.2881, parameterName=EmissiveTextureFalloff})
   B=Add(A=Power(Base=@n48.R, Exp=@n49), B=Power(Base=@n48.R, Exp=@n50))
  )
  n52 = VectorParameter() {parameterName=Emissive Color, defaultValue=<rgba 1.8546 0 4 1>}
  NamedRerouteDeclaration(None=Multiply(A=Power(Base=@n44, Exp=@n47), B=Multiply(A=@n51, B=@n52.RGB))) {name=Emission Out, nodeColor=<rgba 1 1 1 1>}
  NamedRerouteUsage()
 ]
}
